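; Deployed Whisper Remedy Sweet sprite actor, extending parent logic for sleep, wake, death and charge depletion; its ability event is empty here.
; On begin play and wake, the server registers its density tags with the Living World Manager.
; Waking plays an eruption sound; sleeping sets the anim instance's cooldown flag and stops montages.
; Death hides it, depletion spawns tinted despawn Niagara, and overrides enable metal dynamic-destruction effects.
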(event EventExecuteSpriteAbility)

(event Custom|Awoken
  (Audio|Components|Audio|Play (Variables|Default|GetEruptStartSound) 0.0))

(event Custom|RegisterAwokenSprite
  (bind _tags_density (Variables|Dep%28Deletelater%29|GetTags%20Density))
  (Networking|SwitchHasAuthority
    (:Authority
      (if (not (Utilities|Array|IsEmpty _tags_density))
        (Class|FortAthenaLivingWorldManager|RegisterTagDensityReservoir (Class|FortAthenaLivingWorldManager|GetLivingWorldManager) self _tags_density)))))

(event EventBeginPlay
  (bind _tags_density (Variables|Dep%28Deletelater%29|GetTags%20Density))
  ;[1 1 1] Begin Play
  (|CallParentFunction)
  (if (Variables|Default|GetIsSleeping)
    (CallFunction|EnterSleepState))
  (Networking|SwitchHasAuthority
    (:Authority
      (if (not (Utilities|Array|IsEmpty _tags_density))
        (Class|FortAthenaLivingWorldManager|RegisterTagDensityReservoir (Class|FortAthenaLivingWorldManager|GetLivingWorldManager) self _tags_density)))))

(event EventWakeUpSprite
  ;[1 1 1 0.5] On Wake Up
  (|CallParentFunction)
  (CallFunction|RegisterAwokenSprite)
  (CallFunction|Awoken))

(event EventEnterSleepState
  ;[1 1 1 0.5] On Sleep
  (|CallParentFunction)
  (bind _assprite_whisper_remedy_sweet_creature_skeleton_anim_blueprint (Utilities|Casting|CastToSprite_WhisperRemedySweet_Creature_Skeleton_AnimBlueprint (Components|SkeletalMesh|GetAnimInstance (Variables|Default|GetSkeletalMeshComponentBudgeted))))
  (Variables|/Script/SpritesRuntime.SpriteCreatureAnimInstance|SetbIsCooldown true _assprite_whisper_remedy_sweet_creature_skeleton_anim_blueprint)
  (Animation|Montage|MontageStop _assprite_whisper_remedy_sweet_creature_skeleton_anim_blueprint 0.1 0))

(event EventMulticast_HandleBinding
  (params (Actor :type Actor%20Object%20Reference) (HitComp :type Primitive%20Component%20Object%20Reference))
  ;[1 1 1] Handle Binding
  (|CallParentFunction Actor HitComp))

(event Custom|SetLocationFromServer
  (params (ServerLocation :type Vector))
  ;[1 1 1] Set Location From Server
  (Transformation|SetActorLocation self ServerLocation false true)
  (Variables|Default|SetLocation ServerLocation))

(event EventOnDeathServer
  (params (Damage :type Float%20%28single-precision%29) (DamageTags :type Gameplay%20Tag%20Container%20Structure%20%28by%20ref%29) (Momentum :type Vector :default "0, 0, 0") (HitInfo :type Hit%20Result%20Structure%20%28by%20ref%29) (InstigatedBy :type Controller%20Object%20Reference) (DamageCauser :type Actor%20Object%20Reference) (EffectContext :type Gameplay%20Effect%20Context%20Handle%20Structure))
  ;[1 1 1] On Death
  (|CallParentFunction Damage DamageTags Momentum HitInfo InstigatedBy DamageCauser EffectContext)
  (Rendering|SetActorHiddenInGame self true))

(event EventOutOfCharges
  (bind (_returnvalue_location _returnvalue_rotation) (Math|Transform|BreakTransform (Transformation|GetActorTransform)))
  (bind _returnvalue (Niagara|SpawnSystematLocation "/Sprites/FX/Niagara/NS_Sprite_Despawn.NS_Sprite_Despawn" _returnvalue_location _returnvalue_rotation "1.000000,1.000000,1.000000" false true "None" false))
  (Effects|Components|ParticleSystem|SetColorParameter _returnvalue "SpriteColor" "(R=1.891810,G=0.504716,B=2.500000,A=1.000000)"))

(fn GetDynamicDestructionResourceType
  (return "Metal"))

(fn GetDynamicDestructionStaticMeshComponents
  (return (Utilities|Array|MakeArray (Variables|Default|GetStaticMesh)) 0 0))

(fn ShouldUseDynamicDestructionMaterial
  (return true))

(fn ShouldUseDynamicDestructionNiagaraSystem
  (return true))

(fn ConstructionScript
  (|CallParentFunction))
