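; Projectile Blueprint for the Whisper Remedy Sweet sprite, handling the creature's facing and animation while it is in play.
; BeginPlay yaws the skeletal mesh from the instigator's aim plus a right-angle offset and starts a celebrate montage.
; Bounces swap in a ground-bounce montage through a DoOnce gate, with a montage-length delay before the reset call.
; The construction script sets the capsule to overlap a custom game trace channel.
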(event EventBeginPlay
  (bind (_returnvalue_roll _returnvalue_pitch _returnvalue_yaw) (Math|Rotator|BreakRotator (Pawn|GetBaseAimRotation (Variables|Actor|GetInstigator))))
  (|CallParentFunction)
  (if (Utilities|IsValid (Variables|Default|GetSkeletalMesh) :!class /Script/Engine.KismetSystemLibrary)
    (Transformation|SetWorldRotation (Variables|Default|GetSkeletalMesh) (Math|Rotator|MakeRotator 0.0 0.0 (+ _returnvalue_yaw 9e+01)) false false)
    (else
      _))
  (bind _assprite_whisper_remedy_sweet_creature_skeleton_anim_blueprint (Utilities|Casting|CastToSprite_WhisperRemedySweet_Creature_Skeleton_AnimBlueprint (Components|SkeletalMesh|GetAnimInstance (Variables|Default|GetSkeletalMesh))))
  (Animation|Montage|MontagePlay _assprite_whisper_remedy_sweet_creature_skeleton_anim_blueprint "/Sprites_WhisperRemedySweetSprite/Assets/Animation/Backpack_CelebrateA_Water_M.Backpack_CelebrateA_Water_M" 1.0 "MontageLength" 0.0 true))

(event EventOnBounce
  (params (Hit :type Hit%20Result%20Structure%20%28by%20ref%29))
  (|CallParentFunction Hit)
  (bind _assprite_whisper_remedy_sweet_creature_skeleton_anim_blueprint (Utilities|Casting|CastToSprite_WhisperRemedySweet_Creature_Skeleton_AnimBlueprint (Components|SkeletalMesh|GetAnimInstance (Variables|Default|GetSkeletalMesh))))
  (Utilities|DoOnceThenDo false
    (:DoOnce
      (Animation|Montage|MontageStop _assprite_whisper_remedy_sweet_creature_skeleton_anim_blueprint 0.0 (Animation|Montage|GetCurrentActiveMontage _assprite_whisper_remedy_sweet_creature_skeleton_anim_blueprint))
      (bind _returnvalue (Animation|Montage|MontagePlay _assprite_whisper_remedy_sweet_creature_skeleton_anim_blueprint "/Sprites_WhisperRemedySweetSprite/Assets/Animation/Sprite_groundBounce_Water_M.Sprite_groundBounce_Water_M" 1.0 "MontageLength" 0.0 true))
      (Utilities|FlowControl|Delay _returnvalue)
      (CallFunction|ResetDoOnce))))

(event Custom|ResetDoOnce
  (Utilities|DoOnceThenDo false
    (:DoOnce
      (Animation|Montage|MontageStop (stale EventOnBounce _assprite_whisper_remedy_sweet_creature_skeleton_anim_blueprint) 0.0 (Animation|Montage|GetCurrentActiveMontage (stale EventOnBounce _assprite_whisper_remedy_sweet_creature_skeleton_anim_blueprint)))
      (bind _returnvalue (Animation|Montage|MontagePlay (stale EventOnBounce _assprite_whisper_remedy_sweet_creature_skeleton_anim_blueprint) "/Sprites_WhisperRemedySweetSprite/Assets/Animation/Sprite_groundBounce_Water_M.Sprite_groundBounce_Water_M" 1.0 "MontageLength" 0.0 true))
      (Utilities|FlowControl|Delay _returnvalue)
      (CallFunction|ResetDoOnce))))

(fn ConstructionScript
  (|CallParentFunction)
  (Collision|SetCollisionResponsetoChannel (Variables|Collision|GetCapsuleComponent) "ECC_GameTraceChannel18" "ECR_Overlap"))
